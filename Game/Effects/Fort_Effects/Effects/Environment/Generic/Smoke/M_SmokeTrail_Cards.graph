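# Unlit translucent material for smoke trails built from cards, with fake sun lighting, backlight scattering and time-of-day scaling.
# A time-scrolled UV1 mask drives card travel, size and fades, while vertex offset bends cards along a waving tangent toward the camera.
# Panning noise from sliced pseudo-volume textures distorts normals and sphere-masked opacity.
# Opacity also fades by distance, depth and end-on view, with a cheaper path on Low and Medium quality.
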Material /Game/Effects/Fort_Effects/Effects/Environment/Generic/Smoke/M_SmokeTrail_Cards
graph {
 properties {bAutomaticallySetUsageInEditor=false, blendMode=BLEND_Translucent, shadingModel=MSM_Unlit, usedShadingModels=Unlit}
 nodes [
  n1 = Add(A=NamedRerouteUsage(), B=Multiply(A=LocalPosition().XYZ) {constB=-1})
  n2 = Transform(None=WorldPosition() {worldPositionShaderOffset=WPT_CameraRelative}.XYZ) {transformSourceType=TRANSFORMSOURCE_World, transformType=TRANSFORM_Local}
  NamedRerouteDeclaration(None=Normalize(VectorInput=Add(A=@n1, B=@n2))) {name=Card Vector, nodeColor=<rgba 1 0.6797 0 1>}
  n3 = NamedRerouteUsage()
  n4 = NamedRerouteUsage()
  n5 = ScalarParameter() {group=Waving, parameterName=Waving Timescale, sortPriority=0}
  n6 = VectorParameter() {group=Waving, parameterName=Waving Direction, sortPriority=1}
  n7 = Add(
   A=AppendVector(A=Multiply(A=@n3) {constB=0.6}, B=Constant())
   B=Multiply(A=Multiply(A=@n4, B=@n5), B=ComponentMask(None=@n6.RGB) {g=true, r=true}) {constB=-0.15}
  )
  n8 = MaterialFunctionCall!(
   UVs=@n7
   Offset=Constant() {r=0.02}
   Linear Mask=NamedRerouteUsage()
   Non Linear Mask=NamedRerouteUsage()
   Tangent Multiplier=Constant() {r=0.5}
  ) {materialFunction=/Game/Environments/TechArt/SmokeTrail/MF_SmokeTrail_Waving}
  n9 = MaterialFunctionCall(
   UVs=@n7
   Offset=Constant() {r=-0.02}
   Linear Mask=NamedRerouteUsage()
   Non Linear Mask=NamedRerouteUsage()
   Tangent Multiplier=Constant() {r=0.5}
  ) {materialFunction=&MF_SmokeTrail_Waving}
  NamedRerouteDeclaration(None=Normalize(VectorInput=Subtract(A=@n8.Result, B=@n9.Result))) {name=Waving Tangent, nodeColor=<rgba 0 1 0.8594 1>}
  n10 = VectorParameter() {group=Lighting, parameterName=Shadow Color, sortPriority=1, defaultValue=<rgba 0.0869 0.0326 0.1042 1>}
  n11 = NamedRerouteUsage()
  n12 = Add(A=DotProduct(A=AtmosphericLightVector(), B=Multiply(A=@n11) {constB=-1}))
  n13 = Multiply(
   A=Power(
    Base=Multiply(A=@n12) {constB=0.5}
    Exp=ScalarParameter() {defaultValue=4, group=Lighting, parameterName=Backlight Scattering Exponent, sortPriority=3}
   )
   B=ScalarParameter() {group=Lighting, parameterName=Backlight Scattering Intensity, sortPriority=4}
  )
  n14 = VectorParameter() {group=Lighting, parameterName=Highlight Color, sortPriority=0, defaultValue=<rgba 0.1059 0.1059 0.1059 0>}
  n15 = Multiply(A=@n13, B=@n14.RGB)
  NamedRerouteDeclaration(None=Multiply(A=NamedRerouteUsage(), B=LinearInterpolate(A=@n10.RGB, B=@n15, Alpha=@n13))) {name=Backlighting, nodeColor=<rgba 0 1 0.5078 1>}
  n16 = ComponentMask(None=CollectionParameter() {collection=&FortniteMaterialParameters, parameterName=SunLightColor}) {b=true, g=true, r=true}
  n17 = ScalarParameter() {defaultValue=-0.5, group=Lighting, parameterName=Desaturate Sunlight, sortPriority=-1}
  NamedRerouteDeclaration(None=LinearInterpolate(
    A=Desaturation(None=@n16, Fraction=@n17)
    B=Constant3Vector() {constant=<rgba 1 1 1 0>}
    Alpha=ScalarParameter() {group=Lighting, parameterName="Ignore Sunlight  Color", sortPriority=-2}
   )) {name=Sunlight Color, nodeColor=<rgba 0 0.2031 1 1>}
  n18 = Multiply(A=NamedRerouteUsage(), B=ScalarParameter() {defaultValue=0.35, group=Noise, parameterName=Noise Distortion Amplitude, sortPriority=-1})
  n19 = TextureCoordinate()
  n20 = Constant() {r=0.5}
  n21 = ScalarParameter() {defaultValue=0.452, group=Opacity, parameterName=Sphere Mask Radius, sortPriority=8}
  n22 = SphereMask(A=Add(A=ComponentMask(None=@n18) {g=true, r=true}, B=@n19), B=@n20, Radius=@n21, Hardness=Constant())
  n23 = SphereMask(A=@n19, B=@n20, Radius=@n21, Hardness=ScalarParameter() {group=Opacity, parameterName=Sphere Mask Hardness, sortPriority=9})
  n24 = NamedRerouteUsage()
  n25 = OneMinus(None=ScalarParameter() {defaultValue=0.25, group=Opacity, parameterName=Fade Out Length, sliderMax=0.5, sortPriority=7})
  n26 = Multiply(
   A=SmoothStep(Max=ScalarParameter() {defaultValue=0.25, group=Opacity, parameterName=Fade in Length, sliderMax=0.5, sortPriority=6}, Value=@n24) {constMax=0.5}
   B=OneMinus(None=SmoothStep(Min=@n25, Value=@n24) {constMin=0.5})
  )
  n27 = ScalarParameter() {defaultValue=1000, group=CloseFarFader, parameterName=Near Fade Length, sortPriority=0}
  n28 = ScalarParameter() {defaultValue=1, group=CloseFarFader, parameterName=Near Fade Offset, sortPriority=1}
  n29 = ScalarParameter() {defaultValue=25000, group=CloseFarFader, parameterName=Far Fade Length, sortPriority=2}
  n30 = ScalarParameter() {defaultValue=12500, group=CloseFarFader, parameterName=Far Fade Offset, sortPriority=3}
  n31 = ScalarParameter() {defaultValue=-2, group=CloseFarFader, parameterName="Opacity Far Distance Bias (S)", sortPriority=4}
  n32 = Multiply(
   A=Multiply(A=Multiply(A=@n22, B=@n23), B=Saturate(None=@n26))
   B=NamedRerouteDeclaration(None=MaterialFunctionCall!(
     NearFadeLength=@n27
     NearFadeOffset=@n28
     NearFadeCurve
     NearFadeStartingScale
     NearFadeEndingScale
     FarFadeLength=@n29
     FarFadeOffset=@n30
     FarFadeCurve
     FarFadeEndingScale
     Opacity Near Distance Bias
     Opacity Far Distance Bias=@n31
    ) {materialFunction=/Game/Effects/Fort_Effects/Materials/_Functions/MF_SmoothLod}.Opacity) {name=CloseFarOpacity, nodeColor=<rgba 0.2656 0 1 1>}
  )
  n33 = ScalarParameter() {defaultValue=1, group=Opacity, parameterName=Opacity Scalar, sortPriority=0}
  n34 = Power(
   Base=Saturate(None=Multiply(A=@n32, B=@n33))
   Exp=ScalarParameter() {defaultValue=1, group=Opacity, parameterName=Opacity Exponent, sortPriority=2}
  )
  n35 = DepthFade(Opacity=@n34, FadeDistance=ScalarParameter() {defaultValue=10, group=Opacity, parameterName=Depth Fade Distance, sortPriority=3})
  n36 = Multiply(A=ScalarParameter() {defaultValue=2, group=Opacity, parameterName=Opacity Low Quality Scalar, sortPriority=1}, B=@n34)
  n37 = Normalize(VectorInput=VectorParameter() {group=Waving, parameterName=Direction, sortPriority=2, defaultValue=<rgba 0.25 0 3 1>}.RGB)
  n38 = CameraPositionWS()
  n39 = ObjectPositionWS()
  n40 = Normalize(VectorInput=Transform(None=Subtract(A=@n38, B=@n39)) {transformSourceType=TRANSFORMSOURCE_World, transformType=TRANSFORM_Local})
  n41 = ScalarParameter() {defaultValue=0.4, group=Opacity, parameterName=Fade Along Direction Begin, sortPriority=4}
  n42 = Subtract(A=Abs(None=DotProduct(A=@n37, B=@n40)), B=@n41)
  n43 = Max(A=Subtract(A=ScalarParameter() {defaultValue=0.9, group=Opacity, parameterName=Fade Along Direction End, sortPriority=5}, B=@n41)) {constB=0.0001}
  n44 = OneMinus(None=Saturate(None=Divide(A=@n42, B=@n43)))
  NamedRerouteDeclaration(None=Multiply(
    A=QualitySwitch(Default=@n35, Low=@n36, High=@n35, Medium=@n36, Epic=@n35)
    B=NamedRerouteDeclaration(None=@n44) {name=LongAxisFade, nodeColor=<rgba 1 0 0.4922 1>}
   )) {name=Opacity, nodeColor=<rgba 0.0547 0 1 1>}
  n45 = VertexInterpolator(VS=LocalPosition() {localOrigin=InstancePreSkinning}.XYZ)
  n46 = NamedRerouteUsage()
  n47 = ScalarParameter() {defaultValue=1, group=Noise, parameterName=Noise Time Scale}
  n48 = Add(
   A=@n45.PS
   B=Multiply(
    A=Multiply(A=@n46, B=@n47)
    B=VectorParameter() {group=Noise, parameterName=Noise Pan Dir, sortPriority=1, defaultValue=<rgba 0 0 -25 0>}.RGB
   )
  )
  n49 = NamedRerouteUsage()
  n50 = ActorPositionWS()
  n51 = Divide(A=Add(A=Add(A=@n48, B=@n49), B=@n50), B=ScalarParameter() {defaultValue=1024, group=Noise, parameterName=Noise Scale, sortPriority=0}) {constB=200}
  n52 = TextureObject() {samplerType=SAMPLERTYPE_LinearColor, texture=/Game/Environments/Helios/TechArt/LowAltitudeClouds/T_LowAltitudeClouds_N}
  n53 = Constant() {r=8}
  n54 = ScalarParameter() {defaultValue=0.9375, group=Noise, parameterName=AtlasPadding, primitiveDataIndex=8, sortPriority=6}
  n55 = NamedRerouteUsage()
  n56 = Add(
   A=@n45.PS
   B=Multiply(
    A=Multiply(A=@n47, B=@n55)
    B=VectorParameter() {group=Noise, parameterName=Noise Pan Dir Secondary, sortPriority=4, defaultValue=<rgba 0 0 35 0>}.RGB
   )
  )
  n57 = NamedRerouteUsage()
  n58 = Divide(
   A=Add(A=Add(A=@n56, B=@n57), B=@n50)
   B=ScalarParameter() {defaultValue=512, group=Noise, parameterName=Noise Scale Secondary, sortPriority=2}
  ) {constB=200}
  n59 = TextureObject() {samplerType=SAMPLERTYPE_LinearColor, texture=&T_LowAltitudeClouds_N}
  n60 = Constant() {r=8}
  n61 = ScalarParameter() {defaultValue=0.9375, group=Noise, parameterName=AtlasPadding, primitiveDataIndex=8, sortPriority=6}
  n62 = Normalize(VectorInput=LinearInterpolate(
    A=MaterialFunctionCall!(UVs=@n51, Slices Texture=@n52, Atlas Size=@n53, Atlas Padding=@n54) {materialFunction=/Game/Environments/Helios/TechArt/LowAltitudeClouds/MF_3DVolumeFrom2DSlices}.RGB
    B=MaterialFunctionCall(UVs=@n58, Slices Texture=@n59, Atlas Size=@n60, Atlas Padding=@n61) {materialFunction=&MF_3DVolumeFrom2DSlices}.RGB
   ))
  NamedRerouteDeclaration(None=Multiply(A=Add(A=@n62) {constB=-0.5}) {constB=2}) {name=Noise Dir, nodeColor=<rgba 0.8516 1 0 1>}
  MaterialExpressionActorPositionWS_0 = ActorPositionWS()
  n63 = Convert(None=@MaterialExpressionActorPositionWS_0) {
   nodeName=Break Float3
   convertInputs [{
     type=Vector3
     expressionInput {
      expression=&self:MaterialExpressionActorPositionWS_0
      inputName=None
      mask=0
      maskA=0
      maskB=0
      maskG=0
      maskR=0
      outputIndex=0
     }
    }]
   convertOutputs [{type=Scalar}, {type=Scalar}, {type=Scalar}]
  }
  n64 = Add(A=Add(A=@n63, B=@n63), B=@n63)
  n65 = ScalarParameter() {group=Time, parameterName=Time Offset Seed, sortPriority=1}
  NamedRerouteDeclaration(None=Add(A=Time(), B=Multiply(A=@n64, B=@n65))) {name=UniqueTimeOffset, nodeColor=<rgba 0.7109 1 0 1>}
  n66 = Multiply(A=@n14.RGB, B=ScalarParameter() {defaultValue=1, group=Lighting, parameterName=Highlight Intensity, sliderMax=10})
  n67 = Add(A=NamedRerouteUsage(), B=NamedRerouteUsage())
  n68 = DotProduct(
   A=AtmosphericLightVector()
   B=NamedRerouteDeclaration(None=Normalize(VectorInput=@n67)) {name=Distorted Normal, nodeColor=<rgba 0.5703 1 0 1>}
  )
  n69 = Power(Base=Multiply(A=Add(A=@n68)) {constB=0.5}, Exp=ScalarParameter() {defaultValue=1, group=Lighting, parameterName=Highlight Exponent}) {constExponent=4}
  n70 = Add(A=Multiply(A=LinearInterpolate(A=@n10.RGB, B=@n66, Alpha=@n69), B=NamedRerouteUsage()), B=NamedRerouteUsage())
  n71 = MaterialFunctionCall(
   Value_Morning=ScalarParameter() {defaultValue=1, group=Time of Day Scalars, parameterName=Morning, sortPriority=0}
   Value_Day=ScalarParameter() {defaultValue=1, group=Time of Day Scalars, parameterName=Day, sortPriority=1}
   Value_Evening=ScalarParameter() {defaultValue=1, group=Time of Day Scalars, parameterName=Evening, sortPriority=2}
   Value_Night=ScalarParameter() {defaultValue=1, group=Time of Day Scalars, parameterName=Night, sortPriority=3}
  ) {materialFunction=&Scalar_Time_of_Day}
  NamedRerouteDeclaration(None=MaterialFunctionCall(Emissive=Multiply(A=@n70, B=@n71.Result)) {materialFunction=&MF_FX_EmissiveAdjustments}.Result) {name=EmissiveColor, nodeColor=<rgba 1 0 0.3984 1>}
  n72 = TextureCoordinate() {coordinateIndex=1}
  n73 = NamedRerouteUsage()
  n74 = ScalarParameter() {defaultValue=0.05, group=Time, parameterName=Overall Timescale, sortPriority=0}
  n75 = Frac(None=Add(A=ComponentMask(None=@n72) {r=true}, B=Multiply(A=@n73, B=@n74) {constB=0.1}))
  NamedRerouteDeclaration(None=Power(
    Base=NamedRerouteDeclaration(None=@n75) {name=Linear Mask, nodeColor=<rgba 0.9453 0 1 1>}
    Exp=ScalarParameter() {defaultValue=1, group=Shape, parameterName=Shape Exponent, sortPriority=0}
   ) {constExponent=1.6}) {name=Non Linear Mask, nodeColor=<rgba 0.1016 1 0 1>}
  NamedRerouteDeclaration(None=LinearInterpolate(
    A=ScalarParameter() {defaultValue=0.5, group=Cards, parameterName=Width Start, sortPriority=0}
    B=ScalarParameter() {defaultValue=1, group=Cards, parameterName=Width End, sortPriority=1}
    Alpha=NamedRerouteUsage()
   )) {name=Card Width, nodeColor=<rgba 1 0 0.4219 1>}
  n76 = LinearInterpolate(
   A=ScalarParameter() {defaultValue=0.4, group=Cards, parameterName=Size Start, sortPriority=2}
   B=ScalarParameter() {defaultValue=2, group=Cards, parameterName=Size End, sortPriority=3}
   Alpha=NamedRerouteUsage()
  )
  n77 = NamedRerouteUsage()
  NamedRerouteDeclaration(None=Multiply(A=Multiply(A=@n76, B=@n77), B=NamedRerouteUsage())) {name=Card Size, nodeColor=<rgba 0 0.7422 1 1>}
  n78 = NamedRerouteUsage()
  n79 = CrossProduct(A=@n78, B=NamedRerouteUsage())
  n80 = Normalize(VectorInput=@n79)
  n81 = NamedRerouteUsage()
  n82 = Multiply(A=@n80, B=@n81) {constB=0.25}
  n83 = LocalPosition()
  n84 = ComponentMask(None=@n83.XYZ) {r=true}
  n85 = Multiply(A=Multiply(A=NamedRerouteUsage(), B=VertexColor().R)) {constB=-100}
  n86 = NamedRerouteUsage()
  n87 = CrossProduct(A=@n79, B=@n78)
  n88 = Multiply(A=OneMinus(None=@n81), B=@n87)
  n89 = Multiply(A=CrossProduct(A=@n80, B=@n78), B=@n81) {constB=0.25}
  n90 = ComponentMask(None=@n83.XYZ) {g=true}
  n91 = Add(A=Add(A=Multiply(A=@n82, B=@n84), B=Multiply(A=@n85, B=@n86)), B=Multiply(A=Add(A=@n88, B=@n89), B=@n90))
  n92 = Add(A=@n1, B=Multiply(A=@n91, B=NamedRerouteUsage()))
  n93 = Multiply(A=ScalarParameter() {defaultValue=1, group=Cards, parameterName=LOW WPO Scalar}, B=@n92)
  NamedRerouteDeclaration(None=Transform(None=QualitySwitch(Default=@n92, Low=@n93, High=@n92, Medium=@n93, Epic=@n92)) {transformSourceType=TRANSFORMSOURCE_Local}) {name=WPO, nodeColor=<rgba 0.4297 0 1 1>}
  NamedRerouteDeclaration(None=MaterialFunctionCall(UVs=@n7, Offset=Constant(), Linear Mask=NamedRerouteUsage(), Non Linear Mask=NamedRerouteUsage(), Tangent Multiplier=Constant() {r=1}) {materialFunction=&MF_SmokeTrail_Waving}.Result) {name=Waving WPO, nodeColor=<rgba 1 0 0.7969 1>}
  n94 = Multiply(A=@n87, B=@n90)
  n95 = Transform(None=Multiply(A=Subtract(A=@n91, B=@n94)) {constB=0.01}) {transformSourceType=TRANSFORMSOURCE_Local}
  n96 = MaterialFunctionCall() {materialFunction=&ObjectScale}
  NamedRerouteDeclaration(None=VertexInterpolator(VS=Divide(A=@n95, B=@n96.Scale XYZ)).PS) {name=Normals, nodeColor=<rgba 0.4062 0 1 1>}
  NamedRerouteUsage()
  NamedRerouteUsage()
  NamedRerouteUsage()
 ]
}
